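# Physics asset for the Hardwood backpack cosmetic mesh, giving bodies to its three dynamic net bones.
# Each bone has one capsule, shrinking from net1_dyn to net3_dyn.
# Two constraints chain the capsules in sequence, both with the same small swing and twist limit angles.
# No body or constraint ties the chain to any other pack bone.
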PhysicsAsset /BRCosmetics/Accessories/FORT_Backpacks/Hardwood_Backpack/Meshes/M_MED_Hardwood_Pack_PhysicsAsset
bodies {
 net1_dyn [{shapeType=Capsule, length=4.5707, radius=3.0707, center=<0 0.413 -3.274>, rotation=<rot 0 120.0335 -90>}]
 net2_dyn [{shapeType=Capsule, length=2.3948, radius=2.8948, center=<0 0.413 -2.2424>, rotation=<rot 0 106.6751 -90>}]
 net3_dyn [{shapeType=Capsule, length=1.5, radius=1.8917, center=<0 0.413 -1.0928>, rotation=<rot 0 133.7911 -90>}]
}
constraints [
 {bone1Name=net2_dyn, bone2Name=net1_dyn, swing1LimitDegrees=15, swing2LimitDegrees=15, twistLimitDegrees=15}
 {bone1Name=net3_dyn, bone2Name=net2_dyn, swing1LimitDegrees=15, swing2LimitDegrees=15, twistLimitDegrees=15}
]
