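# Animation Blueprint for the Hardwood backpack, adjusting its pose to the wearer's body and adding physics motion.
# The update event reads the owner's clavicle and spine rotations, setting offset alphas during skydive dive or float, otherwise zeroing them and setting a clavicle-driven yaw.
# Chained additive root and pack bone transforms feed a RigidBody node simulating in pack-bone space with damped component motion.
# The spine-driven root tilt node is not connected to the output.
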AnimBlueprint /BRCosmetics/Accessories/FORT_Backpacks/Hardwood_Backpack/Meshes/M_MED_Hardwood_Pack_Skeleton_AnimBP
target_skeleton=/Game/Accessories/FORT_Backpacks/Hardwood_Backpack/Meshes/M_MED_Hardwood_Pack_Skel
graphs [
 AnimGraph {nodes [
   "Transform (Modify) Bone - Bone: root"(Alpha=Get spine_01_alpha() -> (spine_01_alpha:double), Translation="0,-3,0", Rotation="0,0,25") {
    rotationMode=BMM_Additive
    translationMode=BMM_Additive
    translationSpace=BCS_BoneSpace
    alphaBoolBlend {blendOption=Linear}
    boneToModify {boneName=root}
    rotation {roll=25}
    translation {y=-3}
   }
   n1 = "Transform (Modify) Bone - Bone: root"(ComponentPose=Local To Component(LocalPose=Input Pose()), Translation="0,0,0", Rotation=Get clavicle_rotate() -> (clavicle_rotate:Rotator)) {
    rotationMode=BMM_Additive
    scaleMode=BMM_Replace
    scaleSpace=BCS_BoneSpace
    translationMode=BMM_Additive
    alphaBoolBlend {blendOption=Linear}
    boneToModify {boneName=root}
   }
   n2 = "Transform (Modify) Bone - Bone: pack"(ComponentPose=@n1, Translation="0,-2.5,0", Rotation="0,0,0") {
    rotationMode=BMM_Additive
    scaleMode=BMM_Replace
    scaleSpace=BCS_BoneSpace
    translationMode=BMM_Additive
    alphaBoolBlend {blendOption=Linear}
    boneToModify {boneName=pack}
    translation {y=-2.5}
   }
   n3 = "Transform (Modify) Bone - Bone: root"(ComponentPose=@n2, Alpha=Get clavicle_sum_alpha() -> (clavicle_sum_alpha:double), Translation="0,-6,0", Rotation="0,0,0") {
    rotationMode=BMM_Additive
    translationMode=BMM_Additive
    translationSpace=BCS_BoneSpace
    alphaBoolBlend {blendOption=Linear}
    boneToModify {boneName=root}
    translation {y=-6}
   }
   n4 = RigidBody(ComponentPose=@n3) {
    bClampLinearTranslationLimitToRefPose=true
    bOverrideWorldGravity=true
    overridePhysicsAsset=&/M_MED_Hardwood_Pack_PhysicsAsset
    simulationSpace=BaseBoneSpace
    alphaBoolBlend {blendOption=Linear}
    baseBoneRef {boneName=pack}
    componentAppliedLinearAccClamp=<3000 3000 3000>
    componentLinearAccScale=<0.4 0.3 0.3>
    componentLinearVelScale=<0.4 0.3 0.3>
    overrideWorldGravity {z=-980}
   }
   Output Pose(Result=Component To Local(ComponentPose=@n4)) {layerGroup=None}
  ]}
 EventGraph {nodes [
   Event BlueprintUpdateAnimation()
   > n1 = Branch(Condition=OR Boolean(A=Get bIsSkydiveDiveMode(), B=Get bIsSkydiveFloating()))
   n2 = GetOwnerSkeletalMeshComponent()
   n3 = BreakRotator(InRot=BreakTransform(InTransform=GetDeltaTransformFromRefPose(self=@n2, BoneName=clavicle_l, BaseName=spine_01)).Rotation)
   n4 = BreakRotator(InRot=BreakTransform(InTransform=GetDeltaTransformFromRefPose(self=@n2, BoneName=clavicle_r, BaseName=spine_01)).Rotation)
   > Set clavicle_sum_alpha(clavicle_sum_alpha:double=MapRangeClamped(Value=float + float(A=@n3.Yaw, B=@n4.Yaw) {is=title K2Node_CommutativeAssociativeBinaryOperator}, InRangeA=0, InRangeB=-44)) -> (Output_Get:double)
   n5 = BreakRotator(InRot=BreakTransform(InTransform=GetDeltaTransformFromRefPose(self=@n2, BoneName=spine_01, BaseName=root)).Rotation)
   > Set spine_01_alpha(spine_01_alpha:double=MapRangeClamped(Value=@n5.Yaw, InRangeA=90, InRangeB=130)) -> (Output_Get:double)
   n6 = "Absolute (Float)"(A=@n3.Yaw)
   Set clavicle_rotate(execute=@n1.else, clavicle_rotate:Rotator=MakeRotator(Yaw=MapRangeClamped(Value=@n6, InRangeA=-1, InRangeB=25, OutRangeB=-10))) -> (Output_Get:Rotator)
   > Set clavicle_sum_alpha(clavicle_sum_alpha=0)
   > Set spine_01_alpha(spine_01_alpha=0)
   n7 = GetDeltaTransformFromRefPose(self=@n2, BoneName=upperarm_l, BaseName=spine_01)
   float + float(A=@n5.Yaw, B=BreakRotator(InRot=BreakTransform(InTransform=@n7).Rotation).Yaw)
  ]}
]
variables [
 clavicle_rotate {type=Rotator}
 clavicle_sum_alpha {type=number}
 SkelMesh {type=boolean}
 spine_01_alpha {type=number}
]
